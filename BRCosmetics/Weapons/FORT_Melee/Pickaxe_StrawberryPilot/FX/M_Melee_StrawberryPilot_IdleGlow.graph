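# Additive unlit particle material, named as the idle glow for the StrawberryPilot pickaxe, usable on Niagara sprites and meshes.
# A procedural, texture-free soft spot comes from multiplied U and V sines raised to a power, with UV scale and offset parameters.
# That mask is multiplied by particle color, alpha and a view-angle fade.
# Separate branches feed time-of-day-scaled particle color into emissive adjustments and provide a parameterized camera offset.
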Material /BRCosmetics/Weapons/FORT_Melee/Pickaxe_StrawberryPilot/FX/M_Melee_StrawberryPilot_IdleGlow
graph {
 properties {
  bUsedWithNiagaraMeshParticles=true
  bUsedWithNiagaraSprites=true
  bUsedWithParticleSprites=true
  blendMode=BLEND_Additive
  pixelDepthOffsetMode=PDOM_Legacy
  shadingModel=MSM_Unlit
  usedShadingModels=Unlit
 }
 nodes [
  MaterialFunctionCall(Offset Amount=ScalarParameter() {defaultValue=20, parameterName=Camera offset}) {materialFunction=/Engine/Functions/Engine_MaterialFunctions02/WorldPositionOffset/CameraOffset}
  n1 = Constant() {r=1}
  n2 = Constant() {r=0.5}
  n3 = ParticleColor()
  MaterialFunctionCall(Emissive=Multiply(
    A=MaterialFunctionCall(Value_Morning=@n1, Value_Day=@n1, Value_Evening=@n1, Value_Night=@n2) {materialFunction=&Scalar_Time_of_Day}.Result
    B=@n3.RGB
   )) {materialFunction=&MF_FX_EmissiveAdjustments}
  n4 = Multiply(A=TextureCoordinate(), B=ComponentMask(None=VectorParameter() {parameterName=UV_Scale, defaultValue=<rgba 1 1 0 0>}.RGB) {g=true, r=true})
  n5 = ComponentMask(None=VectorParameter() {parameterName=UV_offset}.RGB) {g=true, r=true}
  n6 = Sine(None=Multiply(A=Add(A=@n4, B=@n5)) {constB=0.5})
  n7 = Power(Base=Multiply(A=ComponentMask(None=@n6) {r=true}, B=ComponentMask(None=@n6) {g=true}))
  Multiply(
   A=Multiply(A=@n3.A, B=Multiply(A=@n3.RGB, B=@n7))
   B=MaterialFunctionCall("FadeSensitivity (optional)"=ScalarParameter() {parameterName="FadeSensitivity (optional) (S)"}) {materialFunction=&FadeBasedOnViewAngle}.Result
  )
 ]
}
