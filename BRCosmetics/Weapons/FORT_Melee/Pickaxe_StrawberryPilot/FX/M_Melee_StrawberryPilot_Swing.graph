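# Additive, two-sided unlit Niagara ribbon, sprite and mesh material for the StrawberryPilot pickaxe swing effect, using Kpop contrail textures.
# Three panning, thresholded sine bands form moving highlights that drive both opacity and emissive, tinted by ramp or particle color and dimmed at night.
# A line texture masks opacity, warped by sine waves and a distortion map, with dynamic parameters for offset, speed and mirroring.
# Mobile simplifies emissive and distortion and skips depth fade.
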Material /BRCosmetics/Weapons/FORT_Melee/Pickaxe_StrawberryPilot/FX/M_Melee_StrawberryPilot_Swing
graph {
 properties {
  bUsedWithBeamTrails=true
  bUsedWithNiagaraMeshParticles=true
  bUsedWithNiagaraRibbons=true
  bUsedWithParticleSprites=true
  blendMode=BLEND_Additive
  pixelDepthOffsetMode=PDOM_Legacy
  shadingModel=MSM_Unlit
  twoSided=true
  usedShadingModels=Unlit
 }
 nodes [
  MaterialFunctionCall(Offset Amount=Constant() {r=20}) {materialFunction=/Engine/Functions/Engine_MaterialFunctions02/WorldPositionOffset/CameraOffset}
  n1 = Constant() {r=0.5}
  n2 = Divide(B=OneMinus(None=@n1))
  n3 = MaterialFunctionCall(UVs=TextureCoordinate(), "Rotation Angle (0-1)"=ScalarParameter() {defaultValue=-0.7, group=Specular, parameterName="Specular_rotaion(0-1)"}) {materialFunction=&CustomRotator}
  n4 = DynamicParameter() {paramNames ["offset(0-1)", speed, "reverse(0,1)", Param4]}
  n5 = Add(A=Multiply(A=Time(), B=@n4.speed) {constB=10}) {constB=0}
  n6 = Add(A=@n5)
  n7 = Panner(Coordinate=@n3.Rotated Values, Time=@n6) {speedY=-0.1}
  n8 = Sine(None=OneMinus(None=ComponentMask(None=@n7) {g=true})) {period=1.5}
  n9 = Subtract(A=Multiply(A=@n8, B=@n8), B=@n1) {constB=0.9}
  n10 = Constant() {r=0.8}
  n11 = Divide(B=OneMinus(None=@n10))
  n12 = ComponentMask(None=Panner(Coordinate=@n3.Rotated Values, Time=@n6) {speedY=-0.07}) {g=true}
  n13 = Subtract(A=Sine(None=OneMinus(None=@n12)) {period=0.8}, B=@n10) {constB=0.8}
  n14 = Add(A=Clamp(None=LinearInterpolate(B=@n2, Alpha=@n9) {constB=10}), B=Clamp(None=LinearInterpolate(B=@n11, Alpha=@n13) {constB=5}))
  n15 = Constant() {r=0.995}
  n16 = OneMinus(None=@n15)
  n17 = Panner(Coordinate=@n3.Rotated Values, Time=@n6) {speedY=-0.12}
  n18 = Sine(None=OneMinus(None=ComponentMask(None=@n17) {g=true})) {period=0.5}
  n19 = Clamp(None=LinearInterpolate(B=Divide(B=@n16), Alpha=Subtract(A=@n18, B=@n15) {constB=0.95}) {constB=5})
  n20 = Add(A=Multiply(A=Add(A=@n14, B=@n19), B=ScalarParameter() {defaultValue=1.5, group=Specular, parameterName=Specular_intensity, sortPriority=1}) {constB=2})
  n21 = ParticleColor()
  n22 = TextureCoordinate()
  n23 = Constant() {r=-0.25}
  n24 = Panner(
   Coordinate=Multiply(
    A=MaterialFunctionCall(UVs=@n22, "Rotation Angle (0-1)"=@n23) {materialFunction=&CustomRotator}.Rotated Values
    B=Constant2Vector() {g=1, r=1}
   )
   Time=Time()
  ) {speedY=0.25}
  n25 = StaticSwitch(
   True=@n21.RGB
   False=Multiply(
    A=TextureSampleParameter2D(UVs=@n24) {group=Color, parameterName=Param, texture=/BRCosmetics/Athena/Effects/Textures/Skydive/P_Contrail_Kpop/T_Contrail_Kpop_Ramp}.RGB
    B=@n21.RGB
   )
   Value=StaticBoolParameter() {parameterName=use_paricle_color}
  )
  n26 = Multiply(A=@n20, B=@n25)
  n27 = Multiply(A=@n26) {constB=0.5}
  MaterialFunctionCall(Emissive=ShadingPathSwitch(
    Default=MaterialFunctionCall(Color_Morning=@n26, Color_Day=@n26, Color_Evening=@n26, Color_Night=@n27) {materialFunction=&Color_Time_of_Day}.Result
    Mobile=@n25
   ) {desc=Migrated from FeatureLevelSwitch}) {materialFunction=&MF_FX_EmissiveAdjustments}
  n28 = Constant2Vector() {g=2, r=1}
  n29 = MaterialFunctionCall(UVs=TextureCoordinate(), "Rotation Angle (0-1)"=ScalarParameter() {defaultValue=-0.25, parameterName="UV_rotate(0-1)"}) {materialFunction=&CustomRotator}
  n30 = ComponentMask(None=@n29.Rotated Values) {r=true}
  n31 = AppendVector(A=Abs(None=Subtract(A=@n4."reverse(0,1)", B=@n30)), B=ComponentMask(None=@n29.Rotated Values) {g=true})
  n32 = Constant()
  n33 = ComponentMask(None=@n31) {g=true}
  n34 = Divide(A=OneMinus(None=@n33)) {constB=4}
  n35 = Add(A=@n31, B=AppendVector(A=Constant(), B=@n4."offset(0-1)"))
  n36 = Frac(None=ComponentMask(None=Panner(Coordinate=@n35, Time=@n5) {speedY=0.2}) {g=true})
  n37 = ScalarParameter() {defaultValue=1, group=Distort_map, parameterName=Disort_frequency, sortPriority=4}
  n38 = Panner(Coordinate=Add(A=@n35, B=Constant2Vector() {r=0.2}), Time=@n5) {speedY=0.3}
  n39 = Multiply(A=Frac(None=ComponentMask(None=@n38) {g=true}), B=@n37)
  n40 = Add(A=Sine(None=Multiply(A=@n36, B=@n37)) {period=0.5}, B=Multiply(A=Cosine(None=@n39)) {constB=0.1})
  n41 = ScalarParameter() {defaultValue=0.5, group=Distort_map, parameterName="distort_channel(R-0,G-0.5,B-1)", sortPriority=2}
  n42 = Constant() {r=0.5}
  n43 = VectorParameter() {group=Distort_map, parameterName="distortUV_(R-U,G-V)", sortPriority=1, defaultValue=<rgba 0.1 0.3 0 1>}
  n44 = ComponentMask(None=@n43.RGB) {g=true, r=true}
  n45 = TextureSampleParameter2D(UVs=Panner(Coordinate=Multiply(A=@n31, B=@n44), Time=Time()) {speedY=0.2}) {
   group=Distort_map
   parameterName=Detail_disort_map
   samplerType=SAMPLERTYPE_Normal
   texture=/Game/Global/Textures/Default/Tileable_Patterns/1Normal
  }
  n46 = Add(
   A=@n40
   B=LinearInterpolate(
    B=ScalarParameter() {defaultValue=2, group=Distort_map, parameterName=Distort_lerp, sortPriority=3}
    Alpha=If(A=@n41, B=@n42, "A > B"=@n45.B, "A == B"=@n45.B, "A < B"=@n45.B)
   ) {constB=2}
  ) {constB=0}
  n47 = Multiply(
   A=Multiply(A=Sine(None=@n34), B=ShadingPathSwitch(Default=@n46, Mobile=@n40) {desc=Migrated from FeatureLevelSwitch})
   B=ScalarParameter() {defaultValue=0.15, group=Distort_map, parameterName=Disort_amplitude, sortPriority=5}
  ) {constB=0.15}
  n48 = TextureSampleParameter2D(UVs=Add(A=Multiply(A=@n31, B=@n28), B=StaticSwitchParameter(True=@n47, False=@n32) {defaultValue=true, parameterName=Disort})) {group=Elements, parameterName=Line, texture=/BRCosmetics/Athena/Effects/Textures/Skydive/P_Contrail_Kpop/T_Contrail_Kpop_elements}
  n49 = Multiply(A=OneMinus(None=@n33), B=@n33)
  n50 = Multiply(
   A=If(
    A=ScalarParameter() {group=Elements, parameterName="channel(R-0,G-0.5,B-1)", sortPriority=1}
    B=Constant() {r=0.5}
    "A > B"=@n48.B
    "A == B"=@n48.B
    "A < B"=@n48.B
   )
   B=Power(Base=LinearInterpolate(Alpha=@n49) {constB=4})
  )
  n51 = LinearInterpolate(Alpha=Clamp(None=LinearInterpolate(Alpha=@n33) {constA=-0.5, constB=0.5})) {constB=2}
  n52 = Multiply(A=@n50, B=Multiply(A=@n51, B=@n51))
  n53 = StaticBoolParameter() {parameterName=foot_mask}
  n54 = Multiply(A=@n20, B=Multiply(A=@n21.A, B=StaticSwitch(True=@n52, False=@n50, Value=@n53)))
  ShadingPathSwitch(Default=DepthFade(Opacity=@n54, FadeDistance=ScalarParameter() {defaultValue=30, parameterName=FadeDistance}), Mobile=@n54) {desc=Migrated from FeatureLevelSwitch}
 ]
}
