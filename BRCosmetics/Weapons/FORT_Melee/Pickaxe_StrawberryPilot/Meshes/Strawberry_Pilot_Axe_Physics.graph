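# Physics asset for the Strawberry Pilot pickaxe cosmetic, covering the dynamic mic cord bones of its axe mesh.
# A small capsule body on dyn_micCord is joined to dyn_micCordBase by one constraint with narrow swing and twist limits, allowing restrained dangle.
# The base body pairs a tiny capsule at the bone with a much larger, offset and tilted capsule.
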PhysicsAsset /BRCosmetics/Weapons/FORT_Melee/Pickaxe_StrawberryPilot/Meshes/Strawberry_Pilot_Axe_Physics
bodies {
 dyn_micCord [{shapeType=Capsule, length=5.0796, radius=2.6371, center=<3.7432 -0.0001 0>, rotation=<rot 90 0 0>}]
 dyn_micCordBase [
  {shapeType=Capsule, length=0.505, radius=0.505, rotation=<rot 0 -90 0>}
  {shapeType=Capsule, length=73.1875, radius=29.4375, center=<42.3411 24.4301 -16.5824>, rotation=<rot 71.9304 0 0>}
 ]
}
constraints [{bone1Name=dyn_micCord, bone2Name=dyn_micCordBase, swing1LimitDegrees=20, swing2LimitDegrees=30, twistLimitDegrees=20}]
